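# Translucent unlit Niagara sprite material for an inky void-mask splat that spreads by particle-driven radius and dissolves.
# A randomly rotated inky base texture and sphere mask share UVs warped by flow-map noise, itself swirled by a second panning flow map.
# The levels-adjusted mask loses a dissolve term built from the same noise, then is multiplied by depth fade, which is bypassed on low quality, mobile and forward.
# The KorrBoss switch selects particle color over a constant color.
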Material /Game/Effects/Fort_Effects/Materials/VoidMask/M_VoidMask_Inky_Spread
graph {
 properties {
  bAutomaticallySetUsageInEditor=false
  bUsedWithNiagaraSprites=true
  blendMode=BLEND_Translucent
  shadingModel=MSM_Unlit
  usedShadingModels=Unlit
 }
 nodes [
  n1 = ParticleColor()
  n2 = TextureCoordinate()
  n3 = Multiply(A=DynamicParameter() {defaultValue=<rgba 0.35 0 1 1>, paramNames [Radius, rand, Swirl Scaling, Dissolve]}.rand) {constB=9}
  n4 = Add(
   A=TextureCoordinate() {uTiling=0.2, vTiling=0.3}
   B=DynamicParameter() {defaultValue=<rgba 0.35 0 1 1>, paramNames [Radius, rand, Swirl Scaling, Dissolve]}.rand
  )
  n5 = Subtract(A=ComponentMask(None=TextureSample(UVs=@n4) {samplerType=SAMPLERTYPE_LinearColor, texture=/Game/Effects/Fort_Effects/Creatures/Elemental/Fire/MergedFireFlowMap}.RGB) {g=true, r=true}) {constB=0.5}
  n6 = ScalarParameter() {defaultValue=0.25, parameterName=Base Distortion}
  n7 = TextureCoordinate()
  n8 = VectorParameter() {parameterName=OutlineUV, channelNames {a=Swirl V, b=Swirl U, g=V, r=U}, defaultValue=<rgba 3 3 2 3>}
  n9 = Multiply(A=@n7, B=AppendVector(A=@n8.U, B=@n8.U))
  n10 = Multiply(A=Multiply(A=Time(), B=ScalarParameter() {defaultValue=0.03, parameterName=Pan Speed}), B=Constant2Vector() {g=1, r=0.5})
  n11 = Add(
   A=Multiply(A=@n7, B=AppendVector(A=@n8.Swirl U, B=@n8.Swirl U))
   B=Multiply(A=Time(), B=ScalarParameter() {defaultValue=-0.1, parameterName=Swirl Pan Speed})
  )
  n12 = Subtract(A=ComponentMask(None=TextureSample(UVs=@n11) {samplerType=SAMPLERTYPE_LinearColor, texture=/Game/VisualThreatManager/StormVisuals/LargeClouds/T_Swirl_Flowmap}.RGB) {g=true, r=true}) {constB=0.5}
  n13 = Multiply(
   A=ScalarParameter() {defaultValue=0.78, parameterName=Swirl Strength}
   B=DynamicParameter() {defaultValue=<rgba 0.35 0 1 1>, paramNames [Radius, rand, Swirl Scaling, Dissolve]}.Swirl Scaling
  )
  n14 = TextureSample(UVs=Add(A=Add(A=@n9, B=@n10) {constB=0}, B=Multiply(A=@n12, B=@n13))) {samplerType=SAMPLERTYPE_LinearColor, texture=&MergedFireFlowMap}
  n15 = ComponentMask(None=@n14.RGB) {g=true, r=true}
  n16 = Add(
   A=Add(A=Rotator(Coordinate=@n2, Time=@n3), B=Multiply(A=@n5, B=@n6)) {constB=0}
   B=Multiply(A=Subtract(A=@n15) {constB=0.5}, B=ScalarParameter() {defaultValue=0.4, parameterName=Noise})
  )
  n17 = TextureSample(UVs=@n16) {texture=&/T_VoidMask_Inky_Base}
  n18 = Constant2Vector() {g=0.5, r=0.5}
  n19 = DynamicParameter() {defaultValue=<rgba 0.35 0 1 1>, paramNames [Radius, rand, Swirl Scaling, Dissolve]}
  n20 = Multiply(A=@n1.A, B=Multiply(A=@n17.RGB, B=SphereMask(A=@n16, B=@n18, Radius=@n19.Radius) {hardnessPercent=70}))
  n21 = VectorParameter() {defaultValue=<rgba 0 1 0.7 0.5>}
  n22 = DynamicParameter() {defaultValue=<rgba 0.35 0 1 1>, paramNames [Radius, rand, Swirl Scaling, Dissolve]}
  n23 = Multiply(A=Power(Base=@n14.R) {constExponent=8}, B=ScalarParameter() {defaultValue=150, parameterName=Dissolve Base}) {constB=3}
  Multiply(
   A=NamedRerouteUsage()
   B=Subtract(
    A=MaterialFunctionCall!(
     Texture=@n20
     ---------------
     New Black Value=@n21.R
     New Middle Value=@n21.R
     New White Value=@n21.R
     Middle Point=@n21.R
     "--------------- "
     Define Interpolation Curve
     Interpolation Power
     Invert Interpolation Power
    ) {materialFunction=/Engine/Functions/Engine_MaterialFunctions02/3PointLevels}.Result
    B=Multiply(A=@n22.Dissolve, B=@n23)
   )
  )
  n24 = ScalarParameter() {defaultValue=100, parameterName=DepthFade_Max}
  n25 = Constant() {r=1}
  n26 = QualitySwitch(Default=DepthFade(FadeDistance=LinearInterpolate(A=@n24, Alpha=@n17.R) {constA=100}), Low=@n25)
  NamedRerouteDeclaration(None=ShadingPathSwitch(Default=ShadingPathSwitch(Default=@n26, Mobile=@n25) {desc=Migrated from FeatureLevelSwitch}, Forward=@n25)) {name=Depth Fade, nodeColor=<rgba 1 0.1577 0 1>}
  StaticSwitchParameter(True=@n1.RGB, False=Constant3Vector()) {parameterName=KorrBoss}
  Subtract()
 ]
}
